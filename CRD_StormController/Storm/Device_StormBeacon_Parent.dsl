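; Parent Blueprint for the Storm Beacon creative device in the Storm Controller set, providing shared setup and no storm logic of its own.
; Editor preview and world-ready events call their parents, then initialize the CreativeVisibleinGame component.
; GetVisibilityComponents returns the static mesh component as the only entry in its array.
; Option loads and changes from ToyOptionsComponent both route to OnPropertiesChanged, an empty custom event that child classes can override.
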(event EditorPreview|EventSetupForEditorPreview
  (|CallParentFunction)
  (Class|CreativeVisibleinGameComponent|InitializeComponent (Variables|Default|GetVisibleInGameComponent)))

(event EventOnWorldReady
  (|CallParentFunction)
  (Class|CreativeVisibleinGameComponent|InitializeComponent (Variables|Default|GetVisibleInGameComponent)))

(event OnAnyPropertyChanged(ToyOptionsComponent)
  (CallFunction|OnPropertiesChanged))

(event OnPropertiesLoaded(ToyOptionsComponent)
  (CallFunction|OnPropertiesChanged))

(event Custom|OnPropertiesChanged)

(fn GetVisibilityComponents
  (return (Utilities|Array|MakeArray (Variables|Debug|GetStaticMeshComponent))))

(fn ConstructionScript)
