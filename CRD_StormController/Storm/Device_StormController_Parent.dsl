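; Parent Creative storm controller device that spawns and destroys a storm shield actor during a minigame.
; Authority-only handlers act on minigame start (if enabled), end, and generate/destroy events, keeping one shield with the device's location, yaw and visual overrides.
; Spawning queues storm sickness options on the minigame's mutator, scoped through data registry resolver scopes; destruction and EndPlay dequeue them.
; It also broadcasts phase-end events and leaves property-change and visual-update hooks empty.
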(event MinigameStartedDelegate(FortMinigameLogic)
  (Networking|SwitchHasAuthority
    (:Authority
      (if (Variables|UserOptions|Basic|GetGenerateStormOnGameStart)
        (CallFunction|SpawnStormShield)))))

(event MinigameEndedDelegate(FortMinigameLogic)
  (Networking|SwitchHasAuthority
    (:Authority
      (CallFunction|DestroyStormShield))))

(event EventEndPlay
  (params (EndPlayReason :type EEndPlayReason%20Enum :default "Destroyed"))
  (CallFunction|ClearStormSickness))

(event Custom|GeneratedStorm
  (params (Instigator :type Controller%20Object%20Reference))
  (Networking|SwitchHasAuthority
    (:Authority
      (bind _currentminigame (Variables|/Script/FortniteGame.FortMinigameLogicComponent|GetCurrentMinigame (Variables|Default|GetFortMinigameLogic))
        (:then
          (Minigame|IsInProgress _currentminigame
            (:Yes
              (Variables|Transient|GetStormShield
                (:then)
                (:else
                  (CallFunction|SpawnStormShield))))))
        (:else)))))

(event Custom|DestroyStorm
  (params (Instigator :type Controller%20Object%20Reference))
  (Networking|SwitchHasAuthority
    (:Authority
      (bind _currentminigame (Variables|/Script/FortniteGame.FortMinigameLogicComponent|GetCurrentMinigame (Variables|Default|GetFortMinigameLogic))
        (:then
          (Minigame|IsInProgress _currentminigame
            (:Yes
              (CallFunction|DestroyStormShield))))
        (:else)))))

(event Custom|OnPropertiesChanged)

(event OnAnyPropertyChanged(ToyOptionsComponent)
  (CallFunction|OnPropertiesChanged))

(event OnPropertiesLoaded(ToyOptionsComponent)
  (CallFunction|OnPropertiesChanged))

(event EditorPreview|EventSetupForEditorPreview
  (|CallParentFunction)
  (Class|CreativeVisibleinGameComponent|InitializeComponent (Variables|Default|GetVisibleInGameComponent)))

(event EventOnWorldReady
  (|CallParentFunction)
  (Class|CreativeVisibleinGameComponent|InitializeComponent (Variables|Default|GetVisibleInGameComponent)))

(event Custom|TransmitOnPhaseEnd
  (GameplayEvent|BroadcastEvent (Variables|UserOptions|Events|GetOn%20Phase%20Ended) 0 (Utilities|Struct|MakeGameplayEventwhenPhaseEnded 0)))

(fn BuildDataRegistryResolverScope
  (return (Variables|Default|GetResolver%20Scopes) 1 true))

(fn ClearStormSickness
  (bind _currentminigame (Variables|/Script/FortniteGame.FortMinigameLogicComponent|GetCurrentMinigame (Variables|Default|GetFortMinigameLogic))
    (:then
      (Utilities|IsValid (CallFunction|GetStormSicknessMutator _currentminigame)
        (:Is%20Valid
          (Class|FortAthenaMutatorStormSicknessControl|DequeueStormSickenessOptions (CallFunction|GetStormSicknessMutator _currentminigame) self true))))
    (:else)))

(fn ConvertLocalRadiusToWorldRadius
  (params (LocalRadius :type Float%20%28double-precision%29))
  (bind (_safefromstormplayereffects _instormplayereffects _safefromstormenemyeffects _instormenemyeffects _mapworldscale _replicatedradius _radiusgrowthdata _movedata _radiustoworldscaleconversion) (Class|GetClassDefaults (Variables|Default|GetStormShieldClass)))
  (return (* (* LocalRadius 1e+02) _radiustoworldscaleconversion)))

(fn DestroyStormShield
  (CallFunction|ClearStormSickness)
  (bind _stormshield (Variables|Transient|GetStormShield)
    (:then
      (Actor|DestroyActor _stormshield)
      (Variables|Transient|SetStormShield 0))
    (:else)))

(fn GetCacheContextOverride
  (return 0))

(fn GetStormSicknessMutator
  (params (Minigame :type Fort%20Minigame%20Object%20Reference))
  (bind _returnvalue (Actor|GetComponentbyClass Minigame "/Script/FortniteGame.FortMutatorListComponent"))
  (Utilities|IsValid Minigame
    (:Is%20Valid
      (Utilities|IsValid _returnvalue
        (:Is%20Valid
          (bind _ascreative_mutator_storm_sickness_control (Utilities|Casting|CastToCreativeMutator_StormSicknessControl (Modifier|GetMutatorbyClass _returnvalue "/Game/Athena/Playlists/Creative/Mutators/CreativeMutator_StormSicknessControl.CreativeMutator_StormSicknessControl_C"))
            (:then
              _)
            (:CastFailed
              _))
          (return _ascreative_mutator_storm_sickness_control))
        (:Is%20Not%20Valid
          (return 0))))
    (:Is%20Not%20Valid
      (return 0))))

(fn InitializeResolverScopeArray
  (bind _resolver_scopes (Variables|Default|GetResolver%20Scopes))
  (if (Utilities|Array|IsEmpty _resolver_scopes)
    _
    (elif (== (Utilities|Array|Get(aref) _resolver_scopes) (LocalVariables|GetNameNone))
      _))
  (Variables|Default|SetResolver%20Scopes (Utilities|Array|MakeArray (SpatialGameplay|GetSpatialGameplayInterfaceFNameforActor self))))

(fn SetStormSickness
  (bind _currentminigame (Variables|/Script/FortniteGame.FortMinigameLogicComponent|GetCurrentMinigame (Variables|Default|GetFortMinigameLogic))
    (:then
      (Utilities|IsValid (CallFunction|GetStormSicknessMutator _currentminigame)
        (:Is%20Valid
          (CallFunction|InitializeResolverScopeArray)
          (Class|FortAthenaMutatorStormSicknessControl|EnqueueStormSickenessOptions (CallFunction|GetStormSicknessMutator _currentminigame) (Utilities|Struct|MakeStormSicknessOptions (Variables|UserOptions|Advanced|GetbStormSickness) true self)))))
    (:else)))

(fn SpawnStormShield
  (bind (_returnvalue_roll _returnvalue_pitch _returnvalue_yaw) (Math|Rotator|BreakRotator (Transformation|GetActorRotation)))
  (CallFunction|SetStormSickness)
  (Variables|Transient|GetStormShield
    (:then
      (return (Variables|Transient|GetStormShield)))
    (:else
      (bind _stormshieldclass (Variables|Default|GetStormShieldClass)
        (:then
          (bind _returnvalue (Game|SpawnActorfromClass _stormshieldclass (Math|Transform|MakeTransform (Transformation|GetActorLocation) (Math|Rotator|MakeRotator 0.0 0.0 _returnvalue_yaw)) "AlwaysSpawn" "SelectDefaultAtRuntime" self 0))
          (bind _output_get (Variables|Transient|SetStormShield _returnvalue))
          (bind _asstorm_shield_parent (Utilities|Casting|CastToStormShield_Parent _output_get)
            (:then
              (Class|StormShieldParent|SetStormControllerReference _asstorm_shield_parent self :!class /CRD_StormController/Storm/StormShield_Parent.StormShield_Parent_C)
              (Class|StormShieldParent|ApplyStormVisualOverrides _asstorm_shield_parent (Utilities|Struct|BreakDataTableRowMaterialInterfaceSoft (Variables|UserOptions|Basic|GetStormVisuals))))
            (:CastFailed
              _))
          (return _output_get))
        (:else)))))

(fn UpdateVisuals
  (params (bVisible :type Boolean)))

(fn ConstructionScript
  (CallFunction|InitializeResolverScopeArray))
