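; Creative storm shield for a basic storm controller that forms, holds, then shrinks and/or moves on a schedule passed into Start.
; Authority-side timers chain delay, wait, shrink and move, setting game phase steps along the way.
; The move target uses actor or random yaw at random distance; the finish behavior keeps or destroys the actor.
; RepNotifies derive phase timestamps and drive cosmetic visibility and audio on non-dedicated instances.
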(event Custom|Start
  (params (InDelayTime :type Float%20%28double-precision%29) (InWaitTime :type Float%20%28double-precision%29) (InShrinkTime :type Float%20%28double-precision%29) (InStartRadius :type Float%20%28double-precision%29) (InEndRadius :type Float%20%28double-precision%29) (InLevel :type Integer) (InOnFinishBehavior :type Float%20%28double-precision%29) (InMoveDirection :type E_BasicStormController_MoveDirection%20Enum :default "NewEnumerator0") (InMinMoveDistance :type Float%20%28double-precision%29) (InMaxMoveDistance :type Float%20%28double-precision%29) (InMoveDelayTime :type Float%20%28double-precision%29) (InMoveTime :type Float%20%28double-precision%29))
  (Variables|Default|SetDelayTime InDelayTime)
  (Variables|Default|SetWaitTime InWaitTime)
  (Variables|Default|SetShrinkTime InShrinkTime)
  (Variables|Default|SetStartRadius InStartRadius)
  (Variables|Default|SetEndRadius InEndRadius)
  (Variables|StormShield|SetLevel InLevel)
  (Variables|Default|SetOnFinishBehavior InOnFinishBehavior)
  (Variables|Default|SetMoveDirection InMoveDirection)
  (Variables|Default|SetMinMoveDistance InMinMoveDistance)
  (Variables|Default|SetMaxMoveDistance InMaxMoveDistance)
  (Variables|Default|SetMoveDelayTime InMoveDelayTime)
  (Variables|Default|SetMoveTime InMoveTime)
  (StormShield|InitializeRadiusValue (Variables|Default|GetStartRadius))
  (if (> (Variables|Default|GetDelayTime) 0.0)
    (Utilities|Time|SetTimerbyEvent (EventDispatchers|CreateEvent :!event OnDelayTimerExpired) (Variables|Default|GetDelayTime))
    (StormShield|SetGamePhaseStep "StormForming")
    (else
      (CallFunction|OnDelayTimerExpired))))

(event Custom|Update%20Times
  (Networking|CanExecuteCosmeticEvents
    (:True
      (Variables|FortCreativeStormShield|SetStartWaitTime (+ (Variables|Default|GetDelayTime) (Variables|FortCreativeStormShield|GetInstancedServerTime)))
      (Variables|FortCreativeStormShield|SetStartShrinkTime (+ (Variables|Default|GetWaitTime) (Variables|FortCreativeStormShield|GetStartWaitTime)))
      (Variables|FortCreativeStormShield|SetFinishShrinkTime (+ (Variables|Default|GetShrinkTime) (Variables|FortCreativeStormShield|GetStartShrinkTime))))))

(event StormShield|EventOnRep_GamePhaseStep
  (|CallParentFunction)
  (Networking|CanExecuteCosmeticEvents
    (:True
      (Utilities|FlowControl|Sequence
        (:then_0
          (Development|SetHiddeninGame (Variables|StormShield|GetRoot) (or (< (Variables|StormShield|GetCurrentLocalRadius) 1e+02) (not (or (Utilities|Enum|Equal(Enum) (Variables|FortCreativeStormShield|GetGamePhaseStep) "StormHolding") (Utilities|Enum|Equal(Enum) (Variables|FortCreativeStormShield|GetGamePhaseStep) "StormShrinking")))) true))
        (:then_1
          (switch Utilities|FlowControl|Switch|SwitchonEAthenaGamePhaseStep (Variables|FortCreativeStormShield|GetGamePhaseStep)
            (:StormHolding
              (Audio|PlaySoundatLocation (Variables|Audio|GetBeginHoldingSound) (Variables|StormShield|GetViewTargetLocation))
              (bind _shieldboundaryaudio (Variables|Audio|GetShieldBoundaryAudio)
                (:then
                  (Audio|Components|Audio|SetSound _shieldboundaryaudio (Variables|Audio|GetShieldBoundaryLoopSound)))
                (:else)))
            (:StormShrinking
              (Audio|PlaySoundatLocation (Variables|Audio|GetBeginShrinkingSound) (Variables|StormShield|GetViewTargetLocation))
              (bind _shieldboundaryaudio_2 (Variables|Audio|GetShieldBoundaryAudio)
                (:then
                  (Audio|Components|Audio|SetSound _shieldboundaryaudio_2 (Variables|Audio|GetShieldBoundaryShrinkingLoopSound)))
                (:else)))))))))

(fn OnDelayMoveTimerExpired
  (StormShield|SetTargetLocationtoMoveTo (Variables|FortCreativeStormShield|GetNextTargetLocation) (Variables|Default|GetMoveTime)))

(fn OnDelayTimerExpired
  (bind (_returnvalue_roll _returnvalue_pitch _returnvalue_yaw) (Math|Rotator|BreakRotator (Transformation|GetActorRotation)))
  (Networking|SwitchHasAuthority
    (:Authority
      (if (or (!= (Variables|Default|GetStartRadius) (Variables|Default|GetEndRadius)) (> (Variables|Default|GetMaxMoveDistance) 0.0))
        (if (> (Variables|Default|GetWaitTime) 0.0)
          (Utilities|Time|SetTimerbyEvent (EventDispatchers|CreateEvent :!event OnWaitTimerExpired) (Variables|Default|GetWaitTime))
          (StormShield|SetGamePhaseStep "StormHolding")
          (else
            (CallFunction|OnWaitTimerExpired)))
        (Variables|FortCreativeStormShield|SetNextTargetRadius (Variables|Default|GetEndRadius))
        (Variables|FortCreativeStormShield|SetNextTargetLocation (select (> (Variables|Default|GetMaxMoveDistance) 0.0) (+ (Transformation|GetActorLocation) (* (Math|Vector|GetForwardVector (Math|Rotator|MakeRotator 0.0 0.0 (Utilities|Select :Index (Variables|Default|GetMoveDirection) :NewEnumerator0 _returnvalue_yaw :NewEnumerator1 (Math|Random|RandomFloatinRange 0.0 3.6e+02)))) (Math|Random|RandomFloatinRange (Variables|Default|GetMinMoveDistance) (Variables|Default|GetMaxMoveDistance)))) (Transformation|GetActorLocation)))
        (else
          (StormShield|SetGamePhaseStep "StormHolding"))))))

(fn OnRep_DelayTime)

(fn OnRep_DestroyTime
  (Networking|CanExecuteCosmeticEvents
    (:True
      (if (> (Variables|Default|GetDestroyTime) 0)
        (Variables|FortCreativeStormShield|SetFinishDestroyTime (+ (Variables|Default|GetDestroyTime) (Variables|FortCreativeStormShield|GetFinishShrinkTime)))))))

(fn OnRep_ShrinkTime
  (CallFunction|UpdateTimes))

(fn OnRep_WaitTime
  (CallFunction|UpdateTimes))

(fn OnShrinkTimerExpired
  (Networking|SwitchHasAuthority
    (:Authority
      (StormShield|SetGamePhaseStep "StormHolding")
      (CallFunction|TransmitonPhaseEnd)
      (if (!= (Variables|Default|GetOnFinishBehavior) -1.0)
        (if (== (Variables|Default|GetOnFinishBehavior) 0)
          (Actor|DestroyActor)
          (else
            (Utilities|Time|SetTimerbyEvent (EventDispatchers|CreateEvent :!event K2_DestroyActor) (Math|Float|Max(Float) (Variables|Default|GetOnFinishBehavior) 0.1))
            (Variables|Default|SetDestroyTime (Variables|Default|GetOnFinishBehavior))))))))

(fn OnWaitTimerExpired
  (Networking|SwitchHasAuthority
    (:Authority
      (Utilities|FlowControl|Sequence
        (:then_0
          _)
        (:then_1
          (if (!= (Variables|Default|GetStartRadius) (Variables|Default|GetEndRadius))
            (StormShield|SetTargetRadiusSizeChangeOverGrowthRate (Variables|Default|GetEndRadius) (Variables|Default|GetShrinkTime))
            (Utilities|Time|SetTimerbyEvent (EventDispatchers|CreateEvent :!event OnShrinkTimerExpired) (Variables|Default|GetShrinkTime))))
        (:then_2
          (if (> (Variables|Default|GetMaxMoveDistance) 0.0)
            (if (> (Variables|Default|GetMoveDelayTime) 0.0)
              (Utilities|Time|SetTimerbyEvent (EventDispatchers|CreateEvent :!event OnDelayMoveTimerExpired) (Variables|Default|GetMoveDelayTime))
              (else
                (CallFunction|OnDelayMoveTimerExpired))))))
      (Utilities|FlowControl|DoOnce false
        (:Completed
          (StormShield|SetGamePhaseStep "StormShrinking"))))))

(fn ConstructionScript)
