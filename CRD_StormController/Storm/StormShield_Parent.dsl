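; Parent storm shield Blueprint for the Storm Controller device, providing the boundary's visuals and audio and a link back to its controller.
; On BeginPlay it spawns boundary, holding and holding-tick sounds attached to the shield ring.
; A valid material override sets a variable whose OnRep async-loads and applies it to the ring on cosmetic-capable instances.
; Game phase replication refreshes minigame HUD visibility, and a custom event relays phase end to the stored controller.
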(event EventBeginPlay
  (|CallParentFunction)
  (CallFunction|InitClosestPointAudio))

(event StormShield|EventOnRep_GamePhaseStep
  (Networking|CanExecuteCosmeticEvents
    (:True
      (Class|MinigameHUDFunctionLibrary|UpdateGamePhaseVisibility self))))

(event Custom|TransmitOnPhaseEnd
  (bind _stormcontrollerreference (Variables|Default|GetStormControllerReference)
    (:then
      (Class|DeviceStormControllerParent|TransmitonPhaseEnd _stormcontrollerreference))
    (:else)))

(event Custom|LoadAndApplyCustomBoundaryMaterial
  (bind _object (Utilities|AsyncLoadAsset (Variables|Default|GetCustomBoundaryMaterial))
    (:Completed
      (bind _asmaterial_interface (Utilities|Casting|CastToMaterialInterface _object))
      (Rendering|Material|SetMaterial (Variables|StormShield|GetShieldRing) 0 _asmaterial_interface))))

(fn ApplyStormVisualOverrides
  (params (CustomBoundaryMaterial :type Material%20Interface%20Soft%20Object%20Reference))
  (if (Utilities|IsValidSoftObjectReference CustomBoundaryMaterial)
    (Variables|Default|SetCustomBoundaryMaterial CustomBoundaryMaterial)))

(fn InitClosestPointAudio
  (Utilities|FlowControl|Sequence
    (:then_0
      (bind _returnvalue (Audio|SpawnSoundAttached :Sound (Variables|Audio|GetShieldBoundaryLoopSound) :AttachToComponent (Variables|StormShield|GetShieldRing) :AttachPointName "None" :LocationType "KeepRelativeOffset" :bStopWhenAttachedToDestroyed true :VolumeMultiplier 1.0 :PitchMultiplier 1.0 :StartTime 0.0 :bAutoDestroy false))
      (Variables|Audio|SetShieldBoundaryAudio _returnvalue))
    (:then_1
      (bind _returnvalue_2 (Audio|SpawnSoundAttached :Sound (Variables|Audio|GetHoldingLoopSound) :AttachToComponent (Variables|StormShield|GetShieldRing) :AttachPointName "None" :LocationType "KeepRelativeOffset" :bStopWhenAttachedToDestroyed true :VolumeMultiplier 1.0 :PitchMultiplier 1.0 :StartTime 0.0 :bAutoDestroy false))
      (Variables|Audio|SetHoldingAudio _returnvalue_2))
    (:then_2
      (bind _returnvalue_3 (Audio|SpawnSoundAttached :Sound (Variables|Audio|GetHoldingLoopTickSound) :AttachToComponent (Variables|StormShield|GetShieldRing) :AttachPointName "None" :LocationType "KeepRelativeOffset" :bStopWhenAttachedToDestroyed true :VolumeMultiplier 1.0 :PitchMultiplier 1.0 :StartTime 0.0 :bAutoDestroy false))
      (Variables|Audio|SetHoldingTickAudio _returnvalue_3))))

(fn OnRep_CustomBoundaryMaterial
  (Networking|CanExecuteCosmeticEvents
    (:True
      (CallFunction|LoadandApplyCustomBoundaryMaterial))))

(fn SetStormControllerReference
  (params (StormControllerReference :type Device%20Storm%20Controller%20Parent%20Object%20Reference))
  (Variables|Default|SetStormControllerReference StormControllerReference))

(fn ConstructionScript)
